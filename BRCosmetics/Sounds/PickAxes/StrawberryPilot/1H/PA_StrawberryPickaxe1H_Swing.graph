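# Sound cue for the StrawberryPilot one-handed pickaxe swing, played as a player melee weapon sound.
# Each branch mixes pickaxe swing waves on close attenuation with shared fire-axe distant swings on medium attenuation.
# A quality switch steps down the random variations per branch, with the smallest branch reused for two levels and given wider per-layer pitch modulation.
# Owner-limited concurrency caps overlapping swings, stopping the oldest.
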SoundCue /BRCosmetics/Sounds/PickAxes/StrawberryPilot/1H/PA_StrawberryPickaxe1H_Swing
root=QualityLevel_1
nodes [
 {id=Attenuation_10, children [Modulator_6], properties {attenuationSettings=/Game/Sounds/Attenuation/Fort_Weapon_Pickaxe_Close_Att}}
 {id=Attenuation_11, children [Modulator_7], properties {attenuationSettings=/Game/Sounds/Attenuation/Fort_Weapon_Pickaxe_Medium_Att}}
 {id=Attenuation_6, children [Random_1], properties {attenuationSettings=&Fort_Weapon_Pickaxe_Close_Att}}
 {id=Attenuation_7, children [Random_0], properties {attenuationSettings=&Fort_Weapon_Pickaxe_Medium_Att}}
 {id=Attenuation_8, children [Random_2], properties {attenuationSettings=&Fort_Weapon_Pickaxe_Close_Att}}
 {id=Attenuation_9, children [Random_4], properties {attenuationSettings=&Fort_Weapon_Pickaxe_Medium_Att}}
 {id=Mixer_3, children [Attenuation_6, Attenuation_7], properties {inputVolume [1,1]}}
 {id=Mixer_4, children [Attenuation_8, Attenuation_9], properties {inputVolume [1,1]}}
 {id=Mixer_5, children [Attenuation_10, Attenuation_11], properties {inputVolume [1,1]}}
 {id=Modulator_1, children [Mixer_3]}
 {id=Modulator_5, children [Mixer_4]}
 {id=Modulator_6, children [Random_3], properties {pitchMax=1.2, pitchMin=0.9}}
 {id=Modulator_7, children [WavePlayer_23], properties {pitchMax=1.2, pitchMin=0.9}}
 {id=QualityLevel_1, children [Modulator_1,Modulator_5,Mixer_5,Mixer_5]}
 {id=Random_0, children [WavePlayer_0,WavePlayer_1,WavePlayer_2,WavePlayer_3,WavePlayer_4,WavePlayer_5], properties {weights [1,1,1,1,1,1]}}
 {
  id=Random_1
  children [WavePlayer_6, WavePlayer_7, WavePlayer_8, WavePlayer_10, WavePlayer_11, WavePlayer_12]
  properties {weights [1,1,1,1,1,1]}
 }
 {id=Random_2, children [WavePlayer_13, WavePlayer_14, WavePlayer_15], properties {weights [1,1,1]}}
 {id=Random_3, children [WavePlayer_16, WavePlayer_17], properties {weights [1,1]}}
 {id=Random_4, children [WavePlayer_21, WavePlayer_9, WavePlayer_22], properties {weights [1,1,1]}}
 {id=WavePlayer_0, properties {soundWaveAssetPtr=/Game/Athena/Sounds/Weapons/Fire_Axe_Swing_Distant_04}}
 {id=WavePlayer_1, properties {soundWaveAssetPtr=/Game/Athena/Sounds/Weapons/Fire_Axe_Swing_Distant_01}}
 {id=WavePlayer_10, properties {soundWaveAssetPtr=&/PA_1H_StrawberryPilot_Swing_04}}
 {id=WavePlayer_11, properties {soundWaveAssetPtr=&/PA_1H_StrawberryPilot_Swing_05}}
 {id=WavePlayer_12, properties {soundWaveAssetPtr=&/PA_1H_StrawberryPilot_Swing_06}}
 {id=WavePlayer_13, properties {soundWaveAssetPtr=&/PA_1H_StrawberryPilot_Swing_01}}
 {id=WavePlayer_14, properties {soundWaveAssetPtr=&/PA_1H_StrawberryPilot_Swing_02}}
 {id=WavePlayer_15, properties {soundWaveAssetPtr=&/PA_1H_StrawberryPilot_Swing_03}}
 {id=WavePlayer_16, properties {soundWaveAssetPtr=&PA_1H_StrawberryPilot_Swing_01}}
 {id=WavePlayer_17, properties {soundWaveAssetPtr=&PA_1H_StrawberryPilot_Swing_02}}
 {id=WavePlayer_2, properties {soundWaveAssetPtr=/Game/Athena/Sounds/Weapons/Fire_Axe_Swing_Distant_02}}
 {id=WavePlayer_21, properties {soundWaveAssetPtr=&Fire_Axe_Swing_Distant_04}}
 {id=WavePlayer_22, properties {soundWaveAssetPtr=/Game/Athena/Sounds/Weapons/Fire_Axe_Swing_Distant_06}}
 {id=WavePlayer_23, properties {soundWaveAssetPtr=&Fire_Axe_Swing_Distant_06}}
 {id=WavePlayer_3, properties {soundWaveAssetPtr=/Game/Athena/Sounds/Weapons/Fire_Axe_Swing_Distant_03}}
 {id=WavePlayer_4, properties {soundWaveAssetPtr=/Game/Athena/Sounds/Weapons/Fire_Axe_Swing_Distant_05}}
 {id=WavePlayer_5, properties {soundWaveAssetPtr=&Fire_Axe_Swing_Distant_06}}
 {id=WavePlayer_6, properties {soundWaveAssetPtr=&PA_1H_StrawberryPilot_Swing_01}}
 {id=WavePlayer_7, properties {soundWaveAssetPtr=&PA_1H_StrawberryPilot_Swing_02}}
 {id=WavePlayer_8, properties {soundWaveAssetPtr=&PA_1H_StrawberryPilot_Swing_03}}
 {id=WavePlayer_9, properties {soundWaveAssetPtr=&Fire_Axe_Swing_Distant_03}}
]
properties {
 audioPropertiesSheet=/BRCosmetics/Sounds/Settings/PropertySheets/Pickaxe/APS_Cosmetic_Pickaxe_Swing
 bOverrideConcurrency=true
 bPrimeOnLoad=true
 maxDistance=3000
 soundClassObject=&Fort_WeaponMelee_PSM
 volumeMultiplier=0.5
 concurrencyOverrides {bLimitToOwner=true, maxCount=2, resolutionRule=StopOldest}
 concurrencySet [&Player_Melee_Weapon_Conc]
}
