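# Sound cue for the ready sound of the StrawberryPilot one-handed pickaxe, built from swing samples.
# A quality level node selects among mixes that each layer close StrawberryPilot swings with distant fire axe swings under separate close and medium attenuations.
# Random pools shrink across the quality branches, and the last two slots share one reduced mix with widened pitch modulation.
# Owner-limited concurrency stops the oldest instance on overflow.
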SoundCue /BRCosmetics/Sounds/PickAxes/StrawberryPilot/1H/PA_StrawberryPilot1H_Ready
root=QualityLevel_1
nodes [
 {id=Attenuation_10, children [Modulator_6], properties {attenuationSettings=/Game/Sounds/Attenuation/Fort_Weapon_Pickaxe_Close_Att}}
 {id=Attenuation_11, children [Modulator_7], properties {attenuationSettings=/Game/Sounds/Attenuation/Fort_Weapon_Pickaxe_Medium_Att}}
 {id=Attenuation_6, children [Random_6], properties {attenuationSettings=&Fort_Weapon_Pickaxe_Close_Att}}
 {id=Attenuation_7, children [Random_2], properties {attenuationSettings=&Fort_Weapon_Pickaxe_Medium_Att}}
 {id=Attenuation_8, children [Random_8], properties {attenuationSettings=&Fort_Weapon_Pickaxe_Close_Att}}
 {id=Attenuation_9, children [Random_5], properties {attenuationSettings=&Fort_Weapon_Pickaxe_Medium_Att}}
 {id=Mixer_3, children [Attenuation_6, Attenuation_7], properties {inputVolume [1,1]}}
 {id=Mixer_4, children [Attenuation_8, Attenuation_9], properties {inputVolume [1,1]}}
 {id=Mixer_5, children [Attenuation_10, Attenuation_11], properties {inputVolume [1,1]}}
 {id=Modulator_0, children [Mixer_3]}
 {id=Modulator_5, children [Mixer_4]}
 {id=Modulator_6, children [Random_9], properties {pitchMax=1.2, pitchMin=0.9}}
 {id=Modulator_7, children [WavePlayer_30], properties {pitchMax=1.2, pitchMin=0.9}}
 {id=QualityLevel_1, children [Modulator_0,Modulator_5,Mixer_5,Mixer_5]}
 {
  id=Random_2
  children [WavePlayer_10, WavePlayer_11, WavePlayer_14, WavePlayer_15, WavePlayer_17, WavePlayer_25]
  properties {weights [1,1,1,1,1,1]}
 }
 {id=Random_5, children [WavePlayer_27, WavePlayer_26, WavePlayer_29], properties {weights [1,1,1]}}
 {
  id=Random_6
  children [WavePlayer_31, WavePlayer_32, WavePlayer_33, WavePlayer_34, WavePlayer_35, WavePlayer_36]
  properties {weights [1,1,1,1,1,1]}
 }
 {id=Random_8, children [WavePlayer_37, WavePlayer_38, WavePlayer_39], properties {weights [1,1,1]}}
 {id=Random_9, children [WavePlayer_40, WavePlayer_41], properties {weights [1,1]}}
 {id=WavePlayer_10, properties {soundWaveAssetPtr=/Game/Athena/Sounds/Weapons/Fire_Axe_Swing_Distant_04}}
 {id=WavePlayer_11, properties {soundWaveAssetPtr=/Game/Athena/Sounds/Weapons/Fire_Axe_Swing_Distant_01}}
 {id=WavePlayer_14, properties {soundWaveAssetPtr=/Game/Athena/Sounds/Weapons/Fire_Axe_Swing_Distant_02}}
 {id=WavePlayer_15, properties {soundWaveAssetPtr=/Game/Athena/Sounds/Weapons/Fire_Axe_Swing_Distant_03}}
 {id=WavePlayer_17, properties {soundWaveAssetPtr=/Game/Athena/Sounds/Weapons/Fire_Axe_Swing_Distant_05}}
 {id=WavePlayer_25, properties {soundWaveAssetPtr=/Game/Athena/Sounds/Weapons/Fire_Axe_Swing_Distant_06}}
 {id=WavePlayer_26, properties {soundWaveAssetPtr=&Fire_Axe_Swing_Distant_03}}
 {id=WavePlayer_27, properties {soundWaveAssetPtr=&Fire_Axe_Swing_Distant_04}}
 {id=WavePlayer_29, properties {soundWaveAssetPtr=&Fire_Axe_Swing_Distant_06}}
 {id=WavePlayer_30, properties {soundWaveAssetPtr=&Fire_Axe_Swing_Distant_06}}
 {id=WavePlayer_31, properties {soundWaveAssetPtr=&/PA_1H_StrawberryPilot_Swing_01}}
 {id=WavePlayer_32, properties {soundWaveAssetPtr=&/PA_1H_StrawberryPilot_Swing_02}}
 {id=WavePlayer_33, properties {soundWaveAssetPtr=&/PA_1H_StrawberryPilot_Swing_03}}
 {id=WavePlayer_34, properties {soundWaveAssetPtr=&/PA_1H_StrawberryPilot_Swing_04}}
 {id=WavePlayer_35, properties {soundWaveAssetPtr=&/PA_1H_StrawberryPilot_Swing_05}}
 {id=WavePlayer_36, properties {soundWaveAssetPtr=&/PA_1H_StrawberryPilot_Swing_06}}
 {id=WavePlayer_37, properties {soundWaveAssetPtr=&PA_1H_StrawberryPilot_Swing_01}}
 {id=WavePlayer_38, properties {soundWaveAssetPtr=&PA_1H_StrawberryPilot_Swing_02}}
 {id=WavePlayer_39, properties {soundWaveAssetPtr=&PA_1H_StrawberryPilot_Swing_03}}
 {id=WavePlayer_40, properties {soundWaveAssetPtr=&PA_1H_StrawberryPilot_Swing_01}}
 {id=WavePlayer_41, properties {soundWaveAssetPtr=&PA_1H_StrawberryPilot_Swing_02}}
]
properties {
 attenuationSettings=&Fort_Weapon_Projectile_Close_Athena_Att
 audioPropertiesSheet=/BRCosmetics/Sounds/Settings/PropertySheets/Pickaxe/APS_Cosmetic_Pickaxe_Ready
 bOverrideConcurrency=true
 bPrimeOnLoad=true
 maxDistance=4950
 soundClassObject=&Fort_WeaponMelee
 volumeMultiplier=0.3
 concurrencyOverrides {bLimitToOwner=true, maxCount=2, resolutionRule=StopOldest}
 concurrencySet [&Player_Melee_Weapon_Conc]
}
